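; Parent pawn for simple bird NPCs that flee when hurt or when their perch is threatened.
; The server tracks the movement base actor, fleeing on its death, on a hotfix-tuned chance when it is damaged, or on falls and launch pad events.
; Fleeing clears the Avoid Enemy ability's cooldown, then activates it by tag.
; Flying mode starts attached audio, with a timer feeding smoothed speed into a sound parameter.
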(event OnComponentHit(CapsuleComponent)
  (params (HitComponent :type Primitive%20Component%20Object%20Reference) (OtherActor :type Actor%20Object%20Reference) (OtherComp :type Primitive%20Component%20Object%20Reference) (NormalImpulse :type Vector :default "0, 0, 0") (Hit :type Hit%20Result%20Structure%20%28by%20ref%29))
  (bind (_bblockinghit _binitialoverlap _time _distance _location _impactpoint _normal _impactnormal _physmat _hitactor) (Collision|BreakHitResult Hit))
  ;[1 1 1] Capsule Bump - Possibly update Movement Base Actor and Re-Bind Events
  (Networking|SwitchHasAuthority
    (:Authority
      (bind _movementbaseactor (CallFunction|NPCTryUpdateMovementBaseActor :!pinned true)) ; See if it's our new Base.
      (CallFunction|NPCPawnDebugMessage (Utilities|String|BuildString(Object) (Utilities|String|BuildString(Object) "" "HIT = " _hitactor " | ") "BASE = " _movementbaseactor)))))

(event Custom|NPC%20BindMovementBaseDeath
  (params (MovementBase :type Actor%20Object%20Reference))
  ;[1 1 1] Manage Death of Movement Base Actor - Fly Away
  (bind _asbuilding_actor (Utilities|Casting|CastToBuildingActor MovementBase))
  (EventDispatchers|AssignOnDied _asbuilding_actor (delegate Custom|OnDied_Event_0))
  (EventDispatchers|AssignOnDamaged _asbuilding_actor (delegate Custom|OnDamaged_Event)))

(event Custom|NPC%20UnBindMovementBaseDeath
  (params (MovementBase :type Actor%20Object%20Reference))
  ;[1 1 1] Manage Death of Movement Base Actor - Fly Away
  (bind _asbuilding_actor (Utilities|Casting|CastToBuildingActor MovementBase))
  (EventDispatchers|UnbindEventfromOnDied _asbuilding_actor (delegate Custom|OnDied_Event_0))
  (EventDispatchers|UnbindEventfromOnDamaged _asbuilding_actor (delegate Custom|OnDamaged_Event)))

(event Custom|OnDied_Event_0
  (params (DamagedActor :type Actor%20Object%20Reference) (Damage :type Float%20%28single-precision%29) (InstigatedBy :type Controller%20Object%20Reference) (DamageCauser :type Actor%20Object%20Reference) (HitLocation :type Vector) (FHitComponent :type Primitive%20Component%20Object%20Reference) (BoneName :type Name) (Momentum :type Vector))
  ;[1 1 1] Manage Death of Movement Base Actor - Fly Away
  (bind _data (Utilities|FlowControl|MergeExecutionwithData DamagedActor (stale Custom|OnDamaged_Event DamagedActor))
    (:Then
      (Utilities|FlowControl|DoOnce false
        (:Completed
          (CallFunction|NPCUnBindMovementBaseDeath _data)
          (CallFunction|NPCTryUpdateMovementBaseActor)
          (CallFunction|NPCTrytoFlyAway :!pinned true) ; ...Fly Away
        )))))

(event EventOnDamageServer
  (params (Damage :type Float%20%28single-precision%29) (DamageTags :type Gameplay%20Tag%20Container%20Structure%20%28by%20ref%29) (Momentum :type Vector :default "0, 0, 0") (HitInfo :type Hit%20Result%20Structure%20%28by%20ref%29) (InstigatedBy :type Controller%20Object%20Reference) (DamageCauser :type Actor%20Object%20Reference) (EffectContext :type Gameplay%20Effect%20Context%20Handle%20Structure))
  ;[1 1 1] OnDamageServer - try to Fly Away
  (|CallParentFunction Damage DamageTags Momentum HitInfo InstigatedBy DamageCauser EffectContext)
  (CallFunction|NPCTrytoFlyAway DamageTags))

(event Custom|OnDamaged_Event
  (params (DamagedActor :type Actor%20Object%20Reference) (Damage :type Float%20%28single-precision%29) (InstigatedBy :type Controller%20Object%20Reference) (DamageCauser :type Actor%20Object%20Reference) (HitLocation :type Vector) (FHitComponent :type Primitive%20Component%20Object%20Reference) (BoneName :type Name) (Momentum :type Vector))
  ;[1 1 1] Manage Death of Movement Base Actor - Fly Away
  (bind (_asbool _asintrounded _asfloat) (CallFunction|NPCGetScalableFloatHotfixValue (Variables|Default|GetNPC%20ChanceToFlyAwayWhenMovementBaseIsDamaged%20HF)))
  (if (<= (Math|Random|RandomFloat) _asfloat)
    (bind _data (Utilities|FlowControl|MergeExecutionwithData (stale Custom|OnDied_Event_0 DamagedActor) DamagedActor)
      (:Then
        (Utilities|FlowControl|DoOnce false
          (:Completed
            (CallFunction|NPCUnBindMovementBaseDeath _data)
            (CallFunction|NPCTryUpdateMovementBaseActor)
            (CallFunction|NPCTrytoFlyAway :!pinned true) ; ...Fly Away
          ))))))

(event EventOnDeathPlayEffects
  (params (Damage :type Float%20%28single-precision%29) (DamageTags :type Gameplay%20Tag%20Container%20Structure%20%28by%20ref%29) (Momentum :type Vector :default "0, 0, 0") (HitInfo :type Hit%20Result%20Structure%20%28by%20ref%29) (InstigatedBy :type Fort%20Pawn%20Object%20Reference) (DamageCauser :type Actor%20Object%20Reference) (EffectContext :type Gameplay%20Effect%20Context%20Handle%20Structure))
  (|CallParentFunction Damage DamageTags Momentum HitInfo InstigatedBy DamageCauser EffectContext)
  (CallFunction|StopFlyingAudio))

(event EventBeginPlay
  ;[1 1 1] OnBeginPlay - Setup
  (|CallParentFunction)
  (Networking|SwitchHasAuthority
    (:Authority
      (Variables|Default|SetLastLocation (Transformation|GetActorLocation))
      (bind _movementbaseactor (CallFunction|NPCTryUpdateMovementBaseActor)) ; Track our Movement Base so we can know if it has been destreoyed so we can Fly Away
      (CallFunction|NPCPawnDebugMessage (Utilities|String|BuildString(Object) "SPAWNED! - " "BASE = " _movementbaseactor))
      ;[1 1 1] Track Hits to Traps. We MIGHT use the info to fly away and despawn.
      (Collision|AssignOnActorHit (delegate Custom|OnActorHit_Event))
      ;[1 1 1] Track Gameplay Event we KNOW is launching us. Despawn
      (Ability|Async|WaitGameplayEventToActor 0 "(TagName=\"Event.Used.LaunchPad\")"
        (:EventReceived
          (CallFunction|NPCTrytoFlyAway))))))

(event Custom|OnActorHit_Event
  (params (SelfActor :type Actor%20Object%20Reference) (OtherActor :type Actor%20Object%20Reference) (NormalImpulse :type Vector) (Hit :type Hit%20Result%20Structure%20%28by%20ref%29))
  ;[1 1 1] Track Hits to Traps. We MIGHT use the info to fly away and despawn.
  (Utilities|Casting|CastToBuildingTrap OtherActor)
  (Variables|Default|SetLastTrapHitTime (Utilities|Time|GetGameTimeinSeconds)))

(event EventOnMovementModeChanged
  (params (PrevMovementMode :type EMovementMode%20Enum :default "MOVE_None") (NewMovementMode :type EMovementMode%20Enum :default "MOVE_None") (PrevCustomMode :type Byte) (NewCustomMode :type Byte))
  ;[1 1 1] If we HAD a Base and now we're Falling and our Base is gone, try to Fly Away.
  ; Also if we recently Hit a Trap.
  (|CallParentFunction PrevMovementMode NewMovementMode PrevCustomMode NewCustomMode)
  (Utilities|FlowControl|Sequence
    (:then_0
      (Networking|SwitchHasAuthority
        (:Authority
          (if (and (or (Utilities|Enum|Equal(Enum) PrevMovementMode "MOVE_NavWalking") (Utilities|Enum|Equal(Enum) PrevMovementMode "MOVE_Walking")) (Utilities|Enum|Equal(Enum) NewMovementMode "MOVE_Falling"))
            (bind (_lostbase _movementbaseactor) (CallFunction|NPCHaveWeLostOurMovementBaseActor)) ; We HAD a Base, but now we're Falling and the Base is gone
            (if _lostbase
              (CallFunction|NPCPawnDebugMessage (Utilities|String|BuildString(Object) "FALLING! - " "BASE = " _movementbaseactor))
              (elif (< (- (Utilities|Time|GetGameTimeinSeconds) (Variables|Default|GetLastTrapHitTime)) 0.25)
                _))
            (CallFunction|NPCTrytoFlyAway :!pinned true) ; Fly Away!
          ))))
    (:then_1
      (if (Utilities|Enum|Equal(Enum) NewMovementMode "MOVE_Flying")
        (CallFunction|PlayFlyAudio)
        (else
          (CallFunction|StopFlyingAudio))))))

(event EventOnDeathServer
  (params (Damage :type Float%20%28single-precision%29) (DamageTags :type Gameplay%20Tag%20Container%20Structure%20%28by%20ref%29) (Momentum :type Vector :default "0, 0, 0") (HitInfo :type Hit%20Result%20Structure%20%28by%20ref%29) (InstigatedBy :type Controller%20Object%20Reference) (DamageCauser :type Actor%20Object%20Reference) (EffectContext :type Gameplay%20Effect%20Context%20Handle%20Structure))
  ;[1 1 1] Track Hits to Traps. We MIGHT use the info to fly away and despawn.
  (|CallParentFunction Damage DamageTags Momentum HitInfo InstigatedBy DamageCauser EffectContext)
  (Collision|UnbindEventfromOnActorHit (delegate Custom|OnActorHit_Event)))

(event Custom|LineLoop
  (bind _returnvalue (Transformation|GetActorLocation))
  ;[1 1 1] Debug Draw line as Pawn moves so you can see the path drawn.
  (Utilities|FlowControl|Delay 0.05)
  (Utilities|FlowControl|Sequence
    (:then_0
      (if (> (Math|Vector|VectorLength (Transformation|GetVelocity)) 0.0)
        (Debug|DebugLineSendtoAll (Variables|Default|GetLastLocation) _returnvalue (Utilities|Struct|MakeColor 255 255 255 255) false 1e+01 0 6.0)
        (Variables|Default|SetLastLocation _returnvalue)))
    (:then_1
      (CallFunction|LineLoop))))

(fn NPCClearMovementBaseActor
  (bind _previousmovementbaseactor (Variables|Default|GetPreviousMovementBaseActor)
    (:then
      (CallFunction|NPCUnBindMovementBaseDeath _previousmovementbaseactor) ; UnBind previous Base
      (Variables|Default|SetPreviousMovementBaseActor 0) ; Clear
    )
    (:else)))

(fn NPCHaveWeLostOurMovementBaseActor
  (Variables|Default|GetPreviousMovementBaseActor
    (:then
      (bind _movementbaseactor (CallFunction|NPCTryUpdateMovementBaseActor))
      (Utilities|IsValid _movementbaseactor
        (:Is%20Valid
          (return false (Variables|Default|GetPreviousMovementBaseActor)))
        (:Is%20Not%20Valid
          (return true 0))) ; But now we do not?
    )
    (:else
      (return false 0))) ; We had a Base previously?
)

(fn NPCTryToFlyAway
  (params (DamageTags :type Gameplay%20Tag%20Container%20Structure))
  (bind _aipawnabilitysystemcomponent (Variables|GameplayAbilities|GetAIPawnAbilitySystemComponent))
  (Variables|Default|SetLastDamageTagsAvian DamageTags) ; Save off any damage tags (If appropriate) so we can decide what we might do in the event of EQS failure,\nor (not implemented yet) to skip trying to Fly Away in the first place.
  (if (not (Abilities|HasActiveAbilitywithAnyMatchingTag _aipawnabilitysystemcomponent "(GameplayTags=((TagName=\"NPC.Ability.NonAttack.Avoid.Enemy.Primary\")))"))
    (GameplayEffects|RemoveActiveEffectswithGrantedTags _aipawnabilitysystemcomponent "(GameplayTags=((TagName=\"NPC.Cooldown.Ability.NonAttack.Avoid.Enemy.Primary\")))") ; Remove cooldown (if we had just spawned or flown-in)
    (if (Abilities|CanActivateAbilitywithMatchingTag _aipawnabilitysystemcomponent "(GameplayTags=((TagName=\"NPC.Ability.NonAttack.Avoid.Enemy.Primary\")))")
      (bind _returnvalue (Abilities|TryActivateAbilitiesbyTag (Variables|GameplayAbilities|GetAIPawnAbilitySystemComponent) (GameplayTags|MakeLiteralGameplayTagContainer "(GameplayTags=((TagName=\"NPC.Ability.NonAttack.Avoid.Enemy.Primary\")))") true)) ; Fly Away!
    )))

(fn NPCTryUpdateMovementBaseActor
  (Utilities|FlowControl|Sequence
    (:then_0
      (bind _returnvalue (FortPawn|GetMyMovementBaseActor))
      (if (Utilities|NotEqual(Object) _returnvalue (Variables|Default|GetPreviousMovementBaseActor))
        (CallFunction|NPCUnBindMovementBaseDeath (Variables|Default|GetPreviousMovementBaseActor)) ; UnBind previous Base
        (bind _output_get (Variables|Default|SetPreviousMovementBaseActor _returnvalue)) ; Set New "Previous" Base
        (CallFunction|NPCBindMovementBaseDeath _output_get) ; Bind to NEW Base
      ))
    (:then_1
      (return (Variables|Default|GetPreviousMovementBaseActor)))))

(fn PlayFlyAudio
  (Variables|Default|GetFlyingAudioComponent
    (:then
      _)
    (:else
      (bind _returnvalue (Audio|SpawnSoundAttached (Variables|Sound|GetFlyingSound) (Variables|Transformation|GetRootComponent)))
      (Variables|Default|SetFlyingAudioComponent _returnvalue)))
  (bind _returnvalue_2 (Utilities|Time|SetTimerbyEvent (EventDispatchers|CreateEvent :!event UpdateFlyingAudio) 0.1 true))
  (Variables|Sound|SetHandleFlyingVelocity _returnvalue_2))

(fn StopFlyingAudio
  (Utilities|Time|ClearandInvalidateTimerbyHandle (Variables|Sound|GetHandleFlyingVelocity))
  (bind _flyingaudiocomponent (Variables|Default|GetFlyingAudioComponent)
    (:then
      (if (Audio|Components|Audio|IsPlaying _flyingaudiocomponent)
        (Audio|Components|Audio|FadeOut (Variables|Default|GetFlyingAudioComponent) 0.35 0.0 "Linear" :!class /Script/Engine.AudioComponent)))
    (:else)))

(fn UpdateFlyingAudio
  (bind _output_get (Variables|Sound|SetFlyingVelocity (Math|Interpolation|FInterpTo (Variables|Sound|GetFlyingVelocity) (Math|Vector|VectorLength (Transformation|GetVelocity)) (Utilities|Time|GetWorldDeltaSeconds) 5.0)))
  (Audio|Parameter|SetFloatParameter (Variables|Default|GetFlyingAudioComponent) "Flying" (Math|Float|MapRangeClamped _output_get 0.0 1e+03 0.0 1.5) :!class /Script/AudioExtensions.AudioParameterControllerInterface))

(fn ConstructionScript
  (|CallParentFunction))
